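# Translucent UI material that displays the BlackMonday upsell reward image in challenge UI with an animated, wind-like noise distortion and glow.
# Two layers of panned blurred noise, the first warping the second, form a contrast-adjusted dark-blue-to-white highlight color-dodged onto the brightened image.
# A rotated sine gradient weights the noise UV distortion and an optional vertical bow.
# A ColorTint parameter drives a blue-to-orange tint stage.
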Material /Game/Athena/UI/Challenges/Art/BlackMonday/M_UI_BlackMonday_Reward
graph {
 properties {
  bUsedWithStaticMesh=false
  blendMode=BLEND_Translucent
  materialDomain=MD_UI
  pixelDepthOffsetMode=PDOM_Legacy
  usedShadingModels=Default Lit
 }
 nodes [
  n1 = ScalarParameter() {defaultValue=1.15, parameterName=Noise4_UTile}
  n2 = ScalarParameter() {defaultValue=1.15, parameterName=Noise4_VTile}
  n3 = TextureCoordinate()
  n4 = AppendVector(A=ScalarParameter() {defaultValue=-0.15, parameterName=Noise1_USpeed}, B=ScalarParameter() {defaultValue=0.075, parameterName=Noise1_VSpeed})
  n5 = AppendVector(A=ScalarParameter() {defaultValue=0.2, parameterName=Noise2_USpeed}, B=ScalarParameter() {defaultValue=-0.11, parameterName=Noise2_VSpeed})
  n6 = DotProduct(
   A=TextureSample(UVs=Panner(Coordinate=@n3, Speed=@n4)) {samplerType=SAMPLERTYPE_Grayscale, texture=&LowResBlurredNoise}.R
   B=TextureSample(UVs=Panner(Coordinate=@n3, Speed=@n5)) {samplerType=SAMPLERTYPE_Grayscale, texture=&LowResBlurredNoise}.R
  )
  n7 = Add(A=Multiply(A=@n3, B=AppendVector(A=@n1, B=@n2)), B=Multiply(A=@n6, B=ScalarParameter() {defaultValue=0.5, parameterName=Noise4OffsetAmount}))
  n8 = AppendVector(A=ScalarParameter() {defaultValue=0.12, parameterName=Noise4_USpeed}, B=ScalarParameter() {defaultValue=-0.05, parameterName=Noise4_VSpeed})
  n9 = ScalarParameter() {defaultValue=0.8, parameterName=Noise3_UTile}
  n10 = ScalarParameter() {defaultValue=0.8, parameterName=Noise3_VTile}
  n11 = Add(
   A=Multiply(A=@n3, B=AppendVector(A=@n9, B=@n10))
   B=Multiply(A=@n6, B=ScalarParameter() {defaultValue=0.25, parameterName=Noise3OffsetAmount})
  )
  n12 = AppendVector(A=ScalarParameter() {defaultValue=-0.26, parameterName=Noise3_USpeed}, B=ScalarParameter() {defaultValue=0.11, parameterName=Noise3_VSpeed})
  n13 = DotProduct(
   A=TextureSample(UVs=Panner(Coordinate=@n7, Speed=@n8)) {samplerType=SAMPLERTYPE_Grayscale, texture=&LowResBlurredNoise}.R
   B=TextureSample(UVs=Panner(Coordinate=@n11, Speed=@n12)) {samplerType=SAMPLERTYPE_Grayscale, texture=&LowResBlurredNoise}.R
  )
  n14 = Multiply(A=@n13, B=ScalarParameter() {defaultValue=3, parameterName=NoiseColorStrength})
  n15 = ScalarParameter() {defaultValue=2, parameterName=NoiseContrast}
  n16 = LinearInterpolate(
   A=Constant3Vector() {constant=<rgba 0.02 0.0467 0.1 0>}
   B=Constant3Vector() {constant=<rgba 1 1 1 0>}
   Alpha=Saturate(None=Power(Base=@n14, Exp=@n15))
  )
  n17 = ScalarParameter() {defaultValue=0.75, parameterName=WindColorAmount}
  n18 = Rotator(Coordinate=TextureCoordinate(), Time=ScalarParameter() {defaultValue=-0.1, parameterName=SineRotation}) {speed=6.2832}
  n19 = Constant()
  n20 = ScalarParameter() {defaultValue=0.02, parameterName=SineOffset}
  n21 = Sine(None=Add(A=Multiply(A=@n18) {constB=0.5}, B=AppendVector(A=@n19, B=@n20)))
  n22 = OneMinus(None=Saturate(None=ComponentMask(None=@n21) {g=true}))
  n23 = Multiply(A=@n22, B=ScalarParameter() {defaultValue=8, parameterName=SineGapWidth})
  n24 = Saturate(None=@n23)
  n25 = ScalarParameter() {defaultValue=1.5, parameterName=NoiseStrength}
  n26 = Saturate(None=Multiply(A=@n13, B=Clamp(None=@n23, Max=@n25) {maxDefault=2, minDefault=1}))
  n27 = Add(A=TextureCoordinate(), B=Multiply(A=DotProduct(A=@n24, B=@n26), B=ScalarParameter() {defaultValue=0.035, parameterName=NoiseOffset_Reward}))
  n28 = AppendVector(A=Constant(), B=Multiply(A=@n22, B=ScalarParameter() {parameterName=BowAmount}))
  n29 = Multiply(
   A=TextureSample(UVs=Add(A=@n27, B=@n28)) {texture=&/T_UI_BlackMonday_Upsell_Reward}.RGB
   B=ScalarParameter() {defaultValue=1.5, parameterName=BrightenImage}
  )
  n30 = Saturate(None=MaterialFunctionCall!(Blend=Multiply(A=@n16, B=@n17), Base=Saturate(None=@n29)) {materialFunction=/Engine/Functions/Engine_MaterialFunctions03/Blends/Blend_ColorDodge}.Result)
  n31 = ScalarParameter() {parameterName=ColorTint}
  n32 = LinearInterpolate(A=Constant3Vector() {constant=<rgba 0.05 0.1 0.25 0>}, B=Constant3Vector() {constant=<rgba 1 0.56 0.175 0>}, Alpha=@n31)
  LinearInterpolate(A=@n30, B=@n32, Alpha=@n31)
  MaterialFunctionCall(Blend=@n32, Base=@n30) {materialFunction=&Blend_ColorDodge}
 ]
}
